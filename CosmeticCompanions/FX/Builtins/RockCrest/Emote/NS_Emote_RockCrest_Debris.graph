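# Niagara system for the RockCrest companion emote that sheds dust and rock debris from skeletal mesh bones, with emitters named for elbow and wing-finger bones.
# Each emitter times bursts and spawn rates to the emote, using spawn groups to switch lifetime, motion and renderer visibility between sprite dust and mesh chunks.
# Falling dust is tinted by time of day.
# Fast, heavy-gravity debris gets collision, gated by platform quality and enabled shortly after spawn.
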NiagaraSystem /CosmeticCompanions/FX/Builtins/RockCrest/Emote/NS_Emote_RockCrest_Debris
dependencies {
 usedDataInterfaces [
  &NiagaraDataInterfaceColorCurve
  &NiagaraDataInterfaceCurve
  &NiagaraDataInterfacePlatformSet
  &NiagaraDataInterfaceArrayInt32
  &NiagaraDataInterfaceSkeletalMesh
 ]
 usedDynamicInputs []
 usedModules [
  &SystemState
  &self:SystemSpawnScript.SystemScriptSource.SystemScriptGraph.NiagaraNodeAssignment_0.SetVariables_#1
 ]
}
emitters {
 Elbow_fr_r {
  emitterUpdateScript [
   &EmitterState(Life Cycle Mode=::NewEnumerator1, Loop Behavior=::NewEnumerator1, Loop Duration=0.2, UseLoopDelay=true)
   &SpawnBurst_Instantaneous(Spawn Count=26, Spawn Group=2, Spawn Time=0.05, Use Spawn Probability=false, UseLoopCountLimit=false)
   &SpawnRate(SpawnGroup=1, SpawnRate=35, Use Spawn Probability=false) {module=SpawnRate001}
   &SpawnRate(SpawnRate=25, Use Spawn Probability=false)
   /SetVariables(Emitter.DepthFade=10) {isSetParameters=true, module=SetVariables_#2}
  ]
  particleSpawnScript [
   /SetVariables(Particles.VisibilityTag=&SpawnGroup()) {isSetParameters=true, module=SetVariables_#3}
   &InitializeParticle(
    Color Maximum=<rgba 3.0094 4.1614 5 1>
    Color Minimum=<rgba 3 3 3 1>
    Color Mode=::NewEnumerator2
    Lifetime=&MergeHelper_FloatSwitch(
     0=&RandomRangeFloat(Minimum=1.3, Maximum=1.6)
     1=&RandomRangeFloat(Minimum=0.7, Maximum=1.2)
     2=&RandomRangeFloat(Minimum=0.5, Maximum=1)
     3=?
     Index=&SpawnGroup()
     Num=::NewEnumerator1
    )
    Mesh Scale Max=<0.5 0.5 0.5>
    Mesh Scale Min=<0.1 0.1 0.1>
    Mesh Scale Mode=::NewEnumerator2
    Sprite Rotation Angle=&MergeHelper_FloatSwitch(
     0=180
     1=&RandomRangeFloat(Minimum=0, Maximum=360)
     2=?
     3=?
     Index=&SpawnGroup()
     Num=::NewEnumerator0
    )
    Sprite Rotation Mode=::NewEnumerator0
    Sprite Size Mode=::NewEnumerator3
    Uniform Sprite Size=1
    UsePositionOffset=false
    Write Material Random=true
   )
   /SetVariables(
    Particles.Falling Dust Color:LinearColor=&RandomRangeLinearColor(Minimum=<rgba 0.1771 0.1771 0.1771 1>, Maximum=<rgba 0.2448 0.2448 0.2448 1>)
    Particles.Falling Dust Size:Vector2f=&RandomRangeVector2D(Minimum=<15 28>, Maximum=<15 40>)
    Particles.Falling Dust SubImageIndex:NiagaraFloat=&RandomRangeFloat(Minimum=0, Maximum=3)
   ) {isSetParameters=true, module=SetVariables_#4}
   /SetVariables(
    Particles.Dust Color:LinearColor=<rgba 0.0885 0.0885 0.0885 0.5>
    Particles.Dust Size:Vector2f=&Vector2DFromFloat(Value=&RandomRangeFloat(Minimum=12, Maximum=18))
   ) {isSetParameters=true, module=SetVariables_#5}
   /SetVariables(Particles.MeshIndex=&RandomRangeInt(Minimum=0, Maximum=2)) {isSetParameters=true, module=SetVariables_#6}
   &SkeletalMeshLocation(
    Bone / Socket Index=2
    Bone Sampling Mode=::NewEnumerator4
    MaskSpawnGroup=false
    Orientation Sampling=::NewEnumerator1
    Position Sampling=::NewEnumerator2
    Skeletal Mesh=$System.Skeletal Mesh Reader
   )
   &ShapeLocation(
    Additional Quaternion Rotation=true
    Additional Yaw / Pitch / Roll=<0 0.2 0.03>
    Cylinder Height=13
    Cylinder Radius=0
    MaskSpawnGroup=false
    Non Uniform Scale=<1 2.5 1>
    Offset=&MergeHelper_VectorSwitch(0=<-10 10 -38>, 1=<-10 10 0>, 2=<-10 10 0>, 3=?, Index=&SpawnGroup(), Num=::NewEnumerator1)
    Offset Mode=::NewEnumerator0
    Override Seed=false
    Rotation Mode=::NewEnumerator3
    Rotation Quaternion=$Particles.SkeletalMeshLocation.SampledRotation
    Shape Origin=$Particles.SkeletalMeshLocation.SampledPosition
    Shape Primitive=::NewEnumerator1
    Surface Only=false
    UseOwnerScale=true
   )
   &InitialMeshOrientation(Use Rotation Vector=false)
   &AddVelocity(
    Module.Use Falloff Distance=false
    Origin Offset=<-6.8487 11.7608 -8>
    Override Seed=false
    Velocity Mode=::NewEnumerator1
    Velocity Origin=$Particles.SkeletalMeshLocation.SampledPosition
    Velocity Speed=&MergeHelper_FloatSwitch(
     0=0
     1=0
     2=&RandomRangeFloat(Minimum=50, Maximum=200)
     3=?
     Index=&SpawnGroup()
     Num=::NewEnumerator1
    )
   )
   &CameraOffset(Camera Offset Amount=&MergeHelper_FloatSwitch(0=15, 1=8, 2=0, 3=?, Index=&SpawnGroup(), Num=::NewEnumerator1))
  ]
  particleUpdateScript [
   &ParticleState()
   /SetVariables(
    Particles.Falling Dust Color:LinearColor=&MultiplyLinearColor(
     Linear Color=$Particles.Initial.Falling Dust Color
     Scale Factor=&MakeVector4FromVectorAndFloat(
      "Vector (RGB)"=&MakeVectorFromLinearColorRGB(Color=&MultiplyLinearColor(
        Linear Color=&ColorFromCurve(ColorCurve=DataInterface() {propertyValues {
           AlphaCurve {keys [<at 0 1 RCIM_Cubic>]}
           BlueCurve {keys [<at 0.0061 0.1042>, {interpMode=RCIM_Cubic, leaveTangent=1.7407, time=0.1571, value=0.4323}]}
           GreenCurve {keys [<at 0.0061 0.1042>, {interpMode=RCIM_Cubic, leaveTangent=1.7407, time=0.1571, value=0.4323}]}
           RedCurve {keys [<at 0.0061 0.1042>, {interpMode=RCIM_Cubic, leaveTangent=1.7407, time=0.1571, value=0.4323}]}
          }})
        Scale Factor=&MakeVector4_FromLinearColor(LinearColor=&FNTimeOfDayColor(
          Night Color=<rgba 0.3066 0.4742 0.8177 1>
          Morning Color=<rgba 1 0.3015 0.2916 1>
          Day Color=<rgba 1 1 1 1>
          Evening Color=<rgba 1 0.5153 0.1302 1>
         ))
       ))
      "Float (Alpha)"=&FloatFromCurve(
       FloatCurve=DataInterface() {propertyValues {Curve {keys [
           {interpMode=RCIM_Cubic, leaveTangentWeight=0.0962, tangentMode=RCTM_Break, time=0.05, value=0}
           {interpMode=RCIM_Cubic, tangentMode=RCTM_Break, time=0.2, value=1}
           {arriveTangent=-1.8058, interpMode=RCIM_Cubic, tangentMode=RCTM_Break, time=1, value=0}
          ]}}}
       Scale Curve=&RandomRangeFloat(Minimum=0.4, Maximum=0.7)
      )
     )
    )
    Particles.Falling Dust Dynamic 1:Vector4f=&MakeVector4(
     X=&RandomRangeFloat(Minimum=0, Maximum=1)
     Y=&RandomRangeFloat(Minimum=0.5, Maximum=1)
     Z=3
     W=&FloatFromCurve(FloatCurve=DataInterface() {propertyValues {Curve {keys [<at 0 1 RCIM_Cubic>, {arriveTangent=-1.6019, interpMode=RCIM_Cubic, tangentMode=RCTM_Break, time=1, value=0}]}}})
    )
    Particles.Falling Dust Size:Vector2f=&Multiply_Vector2D(
     A=$Particles.Initial.Falling Dust Size
     B=&Multiply_Vector2D_ByFloat(
      Float=1
      Vector2D=&MakeVector2D(
       X=&FloatFromCurve(FloatCurve=DataInterface() {propertyValues {Curve {keys [
            {arriveTangent=1.3188, interpMode=RCIM_Cubic, time=0, value=0.2}
            {arriveTangent=0.7857, interpMode=RCIM_Cubic, leaveTangent=0.7857, tangentMode=RCTM_User, time=1, value=0.5}
           ]}}})
       Y=&FloatFromCurve(FloatCurve=DataInterface() {propertyValues {Curve {keys [
            {interpMode=RCIM_Cubic, tangentMode=RCTM_User, time=0, value=1.2}
            {arriveTangent=-2.3605, interpMode=RCIM_Cubic, leaveTangent=-2.3605, tangentMode=RCTM_User, time=1, value=0}
           ]}}})
      )
     )
    )
   ) {isSetParameters=true, module=SetVariables_#7}
   /SetVariables(
    Particles.Dust Color:LinearColor=&MultiplyLinearColor(
     Linear Color=$Particles.Initial.Dust Color
     Scale Factor=&MakeVector4FromVectorAndFloat(
      "Vector (RGB)"=<1 1 1>
      "Float (Alpha)"=&FloatFromCurve(FloatCurve=DataInterface() {propertyValues {Curve {keys [<at 0 0 RCIM_Cubic>,<at 0.1 1 RCIM_Cubic>,<at 1 0 RCIM_Cubic>]}}})
     )
    )
    Particles.Dust Dynamic 1:Vector4f=&MakeVector4(X=0, Y=1, Z=0, W=0)
    Particles.Dust Size:Vector2f=&Multiply_Vector2D(
     A=$Particles.Initial.Dust Size
     B=&Vector2DFromFloat(Value=&FloatFromCurve(FloatCurve=DataInterface() {propertyValues {Curve {keys [<at 0 0 RCIM_Cubic>, {arriveTangent=1, interpMode=RCIM_Cubic, leaveTangent=1, time=0.2, value=1}, <at 1 1 RCIM_Cubic>]}}}))
    )
    Particles.Dust SubImageIndex:NiagaraFloat=&FloatFromCurve(FloatCurve=DataInterface() {propertyValues {Curve {keys [<at 0 0>,<at 1 64>]}}})
   ) {isSetParameters=true, module=SetVariables_#8}
   &GravityForce(Gravity=&MergeHelper_VectorSwitch(
     0=<0 0 0>
     1=&RandomRangeVector(Minimum=<0 0 -15>, Maximum=<0 0 -35>)
     2=&RandomRangeVector(Minimum=<0 0 -680>, Maximum=<0 0 -980>)
     3=?
     Index=&SpawnGroup()
     Num=::NewEnumerator1
    ))
   &ScaleMeshSize(Scale Factor=&VectorFromFloat(Value=&FloatFromCurve(FloatCurve=DataInterface() {propertyValues {Curve {keys [
          {arriveTangent=-0.008, interpMode=RCIM_Cubic, leaveTangent=-0.008, tangentMode=RCTM_User, time=0, value=1}
          {arriveTangent=-3.2668, interpMode=RCIM_Cubic, leaveTangent=-3.2668, tangentMode=RCTM_User, time=1, value=0}
         ]}}})))
   &Collision(
    Age Colliding Particles=false
    Analytical Collision Plane Position 1:Vector3f=<0 0 -20>
    Analytical Collision Plane Space:ENiagaraCoordinateSpace=::Local
    CPU Collision Type=::NewEnumerator3
    Collision Enabled=&Boolean_AndOperation(
     Bool1=&IsPlatformSetActive(New Platform Set=DataInterface() {propertyValues {Platforms {qualityLevelMask=-28}}})
     Bool2=&SetBoolByFloatComparison(
      A=&FloatFromCurve(
       FloatCurve=DataInterface() {propertyValues {Curve {
          defaultValue=1
          keys [{interpMode=RCIM_Cubic, tangentMode=RCTM_User, time=0.0045, value=0}, {interpMode=RCIM_Cubic, tangentMode=RCTM_User, time=0.1114, value=1}]
         }}}
       Scale Curve=&MergeHelper_FloatSwitch(0=0, 1=0, 2=1, 3=?, Index=&SpawnGroup(), Num=::NewEnumerator1)
      )
      B=0.5
     )
    )
    KillOnCollision=false
    Module.Control Rest State Via Surface Normal=false
    Module.EnableMaxCollisionCount=false
    Module.Randomize Collision Normal Vector=false
    Perform 2 Analytical Collision Queries=false
   ) {module=Collision001}
   &Collision(
    Age Colliding Particles=false
    Collision Enabled=&Boolean_AndOperation(
     Bool1=&IsPlatformSetActive(New Platform Set=DataInterface() {propertyValues {Platforms {qualityLevelMask=-8}}})
     Bool2=&SetBoolByFloatComparison(
      A=&FloatFromCurve(
       FloatCurve=DataInterface() {propertyValues {Curve {
          defaultValue=1
          keys [{interpMode=RCIM_Cubic, tangentMode=RCTM_User, time=0.0045, value=0}, {interpMode=RCIM_Cubic, tangentMode=RCTM_User, time=0.1114, value=1}]
         }}}
       Scale Curve=&MergeHelper_FloatSwitch(0=0, 1=0, 2=1, 3=?, Index=&SpawnGroup(), Num=::NewEnumerator1)
      )
      B=0.5
     )
    )
    KillOnCollision=false
    Module.Control Rest State Via Surface Normal=false
    Module.EnableMaxCollisionCount=false
    Module.Randomize Collision Normal Vector=false
    Radius Calculation Type=::NewEnumerator1
   )
   &SolveForcesAndVelocity(Clamp Velocity=false, Limit Acceleration=false)
  ]
  rendererClasses [&NiagaraSpriteRendererProperties, &NiagaraMeshRendererProperties]
  renderers [
   {rendererClass=&NiagaraSpriteRendererProperties}
   {rendererClass=&NiagaraSpriteRendererProperties, rendererIndex=1}
   {rendererClass=&NiagaraSpriteRendererProperties, rendererIndex=2}
   {rendererClass=&NiagaraMeshRendererProperties, rendererIndex=3}
  ]
 }
 wing_fingerB_02_l {
  emitterUpdateScript [
   &EmitterState(Life Cycle Mode=::NewEnumerator1, Loop Behavior=::NewEnumerator1, Loop Delay=0.802, Loop Duration=3, UseLoopDelay=true)
   &SpawnRate(
    SpawnRate=&FloatFromCurve(
     FloatCurve=DataInterface() {propertyValues {Curve {keys [<at 0 1 RCIM_Cubic>, {arriveTangent=-0.5841, interpMode=RCIM_Cubic, leaveTangent=-0.5841, time=0.22, value=0.8206}, <at 1.712 0 RCIM_Cubic>]}}}
     CurveIndex=$Emitter.Age
     Scale Curve=30
    )
    Use Spawn Probability=false
   )
   &SpawnBurst_Instantaneous(Spawn Count=8, Spawn Group=1, Spawn Time=0.2, Use Spawn Probability=false, UseLoopCountLimit=false) {module=SpawnBurst_Instantaneous003}
   &SpawnBurst_Instantaneous(Spawn Count=37, Spawn Group=2, Spawn Time=0.88, Use Spawn Probability=false, UseLoopCountLimit=false)
   &SpawnRate(
    SpawnGroup=2
    SpawnRate=&FloatFromCurve(
     FloatCurve=DataInterface() {propertyValues {Curve {keys [
         {arriveTangent=4.5806, interpMode=RCIM_Cubic, leaveTangent=4.5806, tangentMode=RCTM_User, time=1.72, value=0}
         {arriveTangent=-0.011, interpMode=RCIM_Cubic, leaveTangent=-0.011, time=1.9991, value=0.9947}
         {arriveTangent=-10.321, interpMode=RCIM_Cubic, leaveTangent=-10.321, tangentMode=RCTM_User, time=2.2, value=-0.0053}
        ]}}}
     CurveIndex=$Emitter.Age
     Scale Curve=60
    )
    Use Spawn Probability=false
   ) {module=SpawnRate001}
   &SpawnBurst_Instantaneous(Spawn Count=10, Spawn Group=3, Spawn Time=0.88, Use Spawn Probability=false, UseLoopCountLimit=false) {module=SpawnBurst_Instantaneous004}
   &SpawnBurst_Instantaneous(Spawn Count=5, Use Spawn Probability=false, UseLoopCountLimit=false) {enabled=false, module=SpawnBurst_Instantaneous002}
   /SetVariables(Emitter.DepthFade=10) {isSetParameters=true, module=SetVariables_#2}
   &SpawnBurst_Instantaneous(Spawn Count=37, Spawn Time=1.682, Use Spawn Probability=false, UseLoopCountLimit=false) {enabled=false, module=SpawnBurst_Instantaneous001}
  ]
  particleSpawnScript [
   /SetVariables(Particles.VisibilityTag=&SpawnGroup()) {isSetParameters=true, module=SetVariables_#3}
   &InitializeParticle(
    Color Maximum=<rgba 3.0094 4.1614 5 1>
    Color Minimum=<rgba 3 3 3 1>
    Color Mode=::NewEnumerator2
    Lifetime=&MergeHelper_FloatSwitch(
     0=&RandomRangeFloat(Minimum=0.4, Maximum=0.8)
     1=&RandomRangeFloat(Minimum=0.8, Maximum=1.1)
     2=&RandomRangeFloat(Minimum=0.5, Maximum=1)
     3=?
     Index=&SpawnGroup()
     Num=::NewEnumerator1
    )
    Mesh Scale Max=<0.5 0.5 0.5>
    Mesh Scale Min=<0.1 0.1 0.1>
    Mesh Scale Mode=::NewEnumerator2
    Sprite Rotation Angle=&MergeHelper_FloatSwitch(
     0=180
     1=&RandomRangeFloat(Minimum=0, Maximum=360)
     2=?
     3=?
     Index=&SpawnGroup()
     Num=::NewEnumerator0
    )
    Sprite Rotation Mode=::NewEnumerator0
    Sprite Size Mode=::NewEnumerator3
    Uniform Sprite Size=1
    UsePositionOffset=false
    Write Material Random=true
   )
   /SetVariables(
    Particles.Falling Dust Color:LinearColor=&RandomRangeLinearColor(Minimum=<rgba 0.1771 0.1771 0.1771 1>, Maximum=<rgba 0.2448 0.2448 0.2448 1>)
    Particles.Falling Dust Size:Vector2f=&RandomRangeVector2D(Minimum=<30 56>, Maximum=<30 80>)
    Particles.Falling Dust SubImageIndex:NiagaraFloat=&RandomRangeFloat(Minimum=0, Maximum=3)
   ) {isSetParameters=true, module=SetVariables_#4}
   /SetVariables(
    Particles.Dust Color:LinearColor=<rgba 0.0885 0.0885 0.0885 0.25>
    Particles.Dust Size:Vector2f=&Vector2DFromFloat(Value=&MaskFloat(
      Mask Source=0
      Mask Parameter=?
      Mask=1
      Masked Float=&RandomRangeFloat(Minimum=12, Maximum=25)
      Passthrough Float=&RandomRangeFloat(Minimum=30, Maximum=75)
     ))
   ) {isSetParameters=true, module=SetVariables_#5}
   /SetVariables(Particles.MeshIndex=&RandomRangeInt(Minimum=0, Maximum=2)) {isSetParameters=true, module=SetVariables_#6}
   &SkeletalMeshLocation(
    Bone / Socket Index=3
    Bone Sampling Mode=::NewEnumerator4
    MaskSpawnGroup=false
    Position Sampling=::NewEnumerator2
    Skeletal Mesh=$System.Skeletal Mesh Reader
   )
   &ShapeLocation(
    Additional Quaternion Rotation=true
    Cylinder Height=25
    Cylinder Radius=3
    MaskSpawnGroup=false
    Non Uniform Scale=<2 2.5 1>
    Offset=&MergeHelper_VectorSwitch(0=<0 0 0>, 1=<0 0 -5>, 2=<0 0 0>, 3=<0 0 0>, Index=&SpawnGroup(), Num=::NewEnumerator2)
    Offset Mode=::NewEnumerator0
    Override Seed=false
    Rotation Mode=::NewEnumerator3
    Rotation Quaternion=$Particles.SkeletalMeshLocation.SampledRotation
    Shape Origin=$Particles.SkeletalMeshLocation.SampledPosition
    Shape Primitive=::NewEnumerator1
    Surface Only=false
    UseOwnerScale=true
   )
   &InitialMeshOrientation(Use Rotation Vector=false)
   &AddVelocity(
    Module.Use Falloff Distance=false
    Override Seed=false
    Velocity Mode=::NewEnumerator1
    Velocity Speed=&MergeHelper_FloatSwitch(
     0=0
     1=0
     2=&RandomRangeFloat(Minimum=200, Maximum=400)
     3=&RandomRangeFloat(Minimum=25, Maximum=75)
     Index=&SpawnGroup()
     Num=::NewEnumerator2
    )
   )
   &CameraOffset(Camera Offset Amount=&MergeHelper_FloatSwitch(0=0, 1=0, 2=0, 3=?, Index=&SpawnGroup(), Num=::NewEnumerator1))
  ]
  particleUpdateScript [
   &ParticleState()
   /SetVariables(
    Particles.Falling Dust Color:LinearColor=&MultiplyLinearColor(
     Linear Color=$Particles.Initial.Falling Dust Color
     Scale Factor=&MakeVector4FromVectorAndFloat(
      "Vector (RGB)"=&MakeVectorFromLinearColorRGB(Color=&MultiplyLinearColor(
        Linear Color=&ColorFromCurve(ColorCurve=DataInterface() {propertyValues {
           AlphaCurve {keys [<at 0 1 RCIM_Cubic>]}
           BlueCurve {keys [<at 0.0061 0.1042>, {interpMode=RCIM_Cubic, leaveTangent=1.7407, time=0.1571, value=0.4323}]}
           GreenCurve {keys [<at 0.0061 0.1042>, {interpMode=RCIM_Cubic, leaveTangent=1.7407, time=0.1571, value=0.4323}]}
           RedCurve {keys [<at 0.0061 0.1042>, {interpMode=RCIM_Cubic, leaveTangent=1.7407, time=0.1571, value=0.4323}]}
          }})
        Scale Factor=&MakeVector4_FromLinearColor(LinearColor=&FNTimeOfDayColor(
          Night Color=<rgba 0.3066 0.4742 0.8177 1>
          Morning Color=<rgba 1 0.3015 0.2916 1>
          Day Color=<rgba 1 1 1 1>
          Evening Color=<rgba 1 0.5153 0.1302 1>
         ))
       ))
      "Float (Alpha)"=&FloatFromCurve(
       FloatCurve=DataInterface() {propertyValues {Curve {keys [
           {interpMode=RCIM_Cubic, leaveTangentWeight=0.0962, tangentMode=RCTM_Break, time=0.05, value=0}
           {interpMode=RCIM_Cubic, tangentMode=RCTM_Break, time=0.2, value=1}
           {arriveTangent=-1.8058, interpMode=RCIM_Cubic, tangentMode=RCTM_Break, time=1, value=0}
          ]}}}
       Scale Curve=&RandomRangeFloat(Minimum=0.4, Maximum=0.7)
      )
     )
    )
    Particles.Falling Dust Dynamic 1:Vector4f=&MakeVector4(
     X=&RandomRangeFloat(Minimum=0, Maximum=1)
     Y=&RandomRangeFloat(Minimum=0.5, Maximum=1)
     Z=3
     W=&FloatFromCurve(FloatCurve=DataInterface() {propertyValues {Curve {keys [<at 0 1 RCIM_Cubic>, {arriveTangent=-1.6019, interpMode=RCIM_Cubic, tangentMode=RCTM_Break, time=1, value=0}]}}})
    )
    Particles.Falling Dust Size:Vector2f=&Multiply_Vector2D(
     A=$Particles.Initial.Falling Dust Size
     B=&Multiply_Vector2D_ByFloat(
      Float=1
      Vector2D=&MakeVector2D(
       X=&FloatFromCurve(FloatCurve=DataInterface() {propertyValues {Curve {keys [
            {arriveTangent=1.3188, interpMode=RCIM_Cubic, time=0, value=0.2}
            {arriveTangent=0.7857, interpMode=RCIM_Cubic, leaveTangent=0.7857, tangentMode=RCTM_User, time=1, value=0.5}
           ]}}})
       Y=&FloatFromCurve(FloatCurve=DataInterface() {propertyValues {Curve {keys [
            {interpMode=RCIM_Cubic, tangentMode=RCTM_User, time=0, value=1.2}
            {arriveTangent=-2.3605, interpMode=RCIM_Cubic, leaveTangent=-2.3605, tangentMode=RCTM_User, time=1, value=0}
           ]}}})
      )
     )
    )
   ) {isSetParameters=true, module=SetVariables_#7}
   /SetVariables(
    Particles.Dust Color:LinearColor=&MultiplyLinearColor(
     Linear Color=$Particles.Initial.Dust Color
     Scale Factor=&MakeVector4FromVectorAndFloat(
      "Vector (RGB)"=<1 1 1>
      "Float (Alpha)"=&FloatFromCurve(FloatCurve=DataInterface() {propertyValues {Curve {keys [<at 0 0 RCIM_Cubic>,<at 0.3 1 RCIM_Cubic>,<at 1 0 RCIM_Cubic>]}}})
     )
    )
    Particles.Dust Dynamic 1:Vector4f=&MakeVector4(X=0, Y=1, Z=0, W=0)
    Particles.Dust Size:Vector2f=&Multiply_Vector2D(
     A=$Particles.Initial.Dust Size
     B=&Vector2DFromFloat(Value=&FloatFromCurve(FloatCurve=DataInterface() {propertyValues {Curve {keys [<at 0 0 RCIM_Cubic>, {arriveTangent=1, interpMode=RCIM_Cubic, leaveTangent=1, time=0.2, value=1}, <at 1 1 RCIM_Cubic>]}}}))
    )
    Particles.Dust SubImageIndex:NiagaraFloat=&FloatFromCurve(FloatCurve=DataInterface() {propertyValues {Curve {keys [<at 0 48>,<at 1 64>]}}})
   ) {isSetParameters=true, module=SetVariables_#8}
   &GravityForce(Gravity=&MergeHelper_VectorSwitch(
     0=<0 0 0>
     1=&RandomRangeVector(Minimum=<0 0 -50>, Maximum=<0 0 -100>)
     2=&RandomRangeVector(Minimum=<0 0 -680>, Maximum=<0 0 -980>)
     3=?
     Index=&SpawnGroup()
     Num=::NewEnumerator1
    ))
   &ScaleMeshSize(Scale Factor=&VectorFromFloat(Value=&FloatFromCurve(FloatCurve=DataInterface() {propertyValues {Curve {keys [
          {arriveTangent=-0.008, interpMode=RCIM_Cubic, leaveTangent=-0.008, tangentMode=RCTM_User, time=0, value=1}
          {arriveTangent=-3.2668, interpMode=RCIM_Cubic, leaveTangent=-3.2668, tangentMode=RCTM_User, time=1, value=0}
         ]}}})))
   &Collision(
    Age Colliding Particles=false
    Analytical Collision Plane Position 1:Vector3f=<0 0 -20>
    Analytical Collision Plane Space:ENiagaraCoordinateSpace=::Local
    CPU Collision Type=::NewEnumerator3
    Collision Enabled=&Boolean_AndOperation(
     Bool1=&IsPlatformSetActive(New Platform Set=DataInterface() {propertyValues {Platforms {qualityLevelMask=-28}}})
     Bool2=&SetBoolByFloatComparison(
      A=&FloatFromCurve(
       FloatCurve=DataInterface() {propertyValues {Curve {
          defaultValue=1
          keys [{interpMode=RCIM_Cubic, tangentMode=RCTM_User, time=0.0045, value=0}, {interpMode=RCIM_Cubic, tangentMode=RCTM_User, time=0.1114, value=1}]
         }}}
       Scale Curve=&MergeHelper_FloatSwitch(0=0, 1=0, 2=1, 3=?, Index=&SpawnGroup(), Num=::NewEnumerator1)
      )
      B=0.5
     )
    )
    KillOnCollision=false
    Module.Control Rest State Via Surface Normal=false
    Module.EnableMaxCollisionCount=false
    Module.Randomize Collision Normal Vector=false
    Perform 2 Analytical Collision Queries=false
   ) {module=Collision001}
   &Collision(
    Age Colliding Particles=false
    Collision Enabled=&Boolean_AndOperation(
     Bool1=&IsPlatformSetActive(New Platform Set=DataInterface() {propertyValues {Platforms {qualityLevelMask=-8}}})
     Bool2=&SetBoolByFloatComparison(
      A=&FloatFromCurve(
       FloatCurve=DataInterface() {propertyValues {Curve {
          defaultValue=1
          keys [{interpMode=RCIM_Cubic, tangentMode=RCTM_User, time=0.0045, value=0}, {interpMode=RCIM_Cubic, tangentMode=RCTM_User, time=0.1114, value=1}]
         }}}
       Scale Curve=&MergeHelper_FloatSwitch(0=0, 1=0, 2=1, 3=?, Index=&SpawnGroup(), Num=::NewEnumerator1)
      )
      B=0.5
     )
    )
    KillOnCollision=false
    Module.Control Rest State Via Surface Normal=false
    Module.EnableMaxCollisionCount=false
    Module.Randomize Collision Normal Vector=false
    Radius Calculation Type=::NewEnumerator1
   )
   &Drag(
    Drag=&MergeHelper_FloatSwitch(0=0, 1=1, 2=0, 3=1, Index=&SpawnGroup(), Num=::NewEnumerator2)
    Use Linear Drag=true
    Use Rotational Drag=true
   )
   &SolveForcesAndVelocity(Clamp Velocity=false, Limit Acceleration=false)
  ]
  rendererClasses [&NiagaraSpriteRendererProperties, &NiagaraMeshRendererProperties]
  renderers [
   {rendererClass=&NiagaraSpriteRendererProperties}
   {rendererClass=&NiagaraSpriteRendererProperties, rendererIndex=1}
   {rendererClass=&NiagaraSpriteRendererProperties, rendererIndex=2}
   {rendererClass=&NiagaraMeshRendererProperties, rendererIndex=3}
   {rendererClass=&NiagaraSpriteRendererProperties, rendererIndex=4}
  ]
 }
 wing_fingerB_02_r {
  emitterUpdateScript [
   &EmitterState(Life Cycle Mode=::NewEnumerator1, Loop Behavior=::NewEnumerator1, Loop Delay=0.802, Loop Duration=3, UseLoopDelay=true)
   &SpawnRate(
    SpawnRate=&FloatFromCurve(
     FloatCurve=DataInterface() {propertyValues {Curve {keys [
         {arriveTangent=-0.8154, interpMode=RCIM_Cubic, time=0.2908, value=0}
         {arriveTangent=3.5805, interpMode=RCIM_Cubic, leaveTangent=3.5805, time=0.3, value=1}
         {arriveTangent=-0.7082, interpMode=RCIM_Cubic, leaveTangent=-0.7082, time=0.52, value=0.8206}
         <at 1.712 0 RCIM_Cubic>
        ]}}}
     CurveIndex=$Emitter.Age
     Scale Curve=30
    )
    Use Spawn Probability=false
   )
   &SpawnBurst_Instantaneous(Spawn Count=8, Spawn Group=1, Spawn Time=0.3, Use Spawn Probability=false, UseLoopCountLimit=false) {module=SpawnBurst_Instantaneous003}
   &SpawnBurst_Instantaneous(Spawn Count=18, Spawn Group=2, Spawn Time=1.34, Use Spawn Probability=false, UseLoopCountLimit=false)
   &SpawnRate(
    SpawnGroup=2
    SpawnRate=&FloatFromCurve(
     FloatCurve=DataInterface() {propertyValues {Curve {keys [
         {arriveTangent=4.5806, interpMode=RCIM_Cubic, leaveTangent=4.5806, tangentMode=RCTM_User, time=2.242, value=0}
         <at 2.3939 0.9855 RCIM_Cubic>
         {arriveTangent=-10.321, interpMode=RCIM_Cubic, leaveTangent=-10.321, tangentMode=RCTM_User, time=2.549, value=0}
        ]}}}
     CurveIndex=$Emitter.Age
     Scale Curve=60
    )
    Use Spawn Probability=false
   ) {module=SpawnRate001}
   &SpawnBurst_Instantaneous(Spawn Count=10, Spawn Group=3, Spawn Time=1.3, Use Spawn Probability=false, UseLoopCountLimit=false) {module=SpawnBurst_Instantaneous004}
   &SpawnBurst_Instantaneous(Spawn Count=5, Use Spawn Probability=false, UseLoopCountLimit=false) {enabled=false, module=SpawnBurst_Instantaneous002}
   /SetVariables(Emitter.DepthFade=10) {isSetParameters=true, module=SetVariables_#2}
   &SpawnBurst_Instantaneous(Spawn Count=37, Spawn Time=1.682, Use Spawn Probability=false, UseLoopCountLimit=false) {enabled=false, module=SpawnBurst_Instantaneous001}
  ]
  particleSpawnScript [
   /SetVariables(Particles.VisibilityTag=&SpawnGroup()) {isSetParameters=true, module=SetVariables_#3}
   &InitializeParticle(
    Color Maximum=<rgba 3.0094 4.1614 5 1>
    Color Minimum=<rgba 3 3 3 1>
    Color Mode=::NewEnumerator2
    Lifetime=&MergeHelper_FloatSwitch(
     0=&RandomRangeFloat(Minimum=0.4, Maximum=0.8)
     1=&RandomRangeFloat(Minimum=0.8, Maximum=1.1)
     2=&RandomRangeFloat(Minimum=0.5, Maximum=1)
     3=?
     Index=&SpawnGroup()
     Num=::NewEnumerator1
    )
    Mesh Scale Max=<0.5 0.5 0.5>
    Mesh Scale Min=<0.1 0.1 0.1>
    Mesh Scale Mode=::NewEnumerator2
    Sprite Rotation Angle=&MergeHelper_FloatSwitch(
     0=180
     1=&RandomRangeFloat(Minimum=0, Maximum=360)
     2=?
     3=?
     Index=&SpawnGroup()
     Num=::NewEnumerator0
    )
    Sprite Rotation Mode=::NewEnumerator0
    Sprite Size Mode=::NewEnumerator3
    Uniform Sprite Size=1
    UsePositionOffset=false
    Write Material Random=true
   )
   /SetVariables(
    Particles.Falling Dust Color:LinearColor=&RandomRangeLinearColor(Minimum=<rgba 0.1771 0.1771 0.1771 1>, Maximum=<rgba 0.2448 0.2448 0.2448 1>)
    Particles.Falling Dust Size:Vector2f=&RandomRangeVector2D(Minimum=<30 56>, Maximum=<30 80>)
    Particles.Falling Dust SubImageIndex:NiagaraFloat=&RandomRangeFloat(Minimum=0, Maximum=3)
   ) {isSetParameters=true, module=SetVariables_#4}
   /SetVariables(
    Particles.Dust Color:LinearColor=<rgba 0.0885 0.0885 0.0885 0.25>
    Particles.Dust Size:Vector2f=&Vector2DFromFloat(Value=&MaskFloat(
      Mask Source=0
      Mask Parameter=?
      Mask=1
      Masked Float=&RandomRangeFloat(Minimum=12, Maximum=25)
      Passthrough Float=&RandomRangeFloat(Minimum=30, Maximum=50)
     ))
   ) {isSetParameters=true, module=SetVariables_#5}
   /SetVariables(Particles.MeshIndex=&RandomRangeInt(Minimum=0, Maximum=2)) {isSetParameters=true, module=SetVariables_#6}
   &SkeletalMeshLocation(
    Bone / Socket Index=4
    Bone Sampling Mode=::NewEnumerator4
    MaskSpawnGroup=false
    Position Sampling=::NewEnumerator2
    Skeletal Mesh=$System.Skeletal Mesh Reader
   )
   &ShapeLocation(
    Additional Quaternion Rotation=true
    Cylinder Height=25
    Cylinder Radius=3
    MaskSpawnGroup=false
    Non Uniform Scale=<2 2.5 1>
    Offset=&MergeHelper_VectorSwitch(0=<0 0 0>, 1=<0 0 -5>, 2=<0 0 0>, 3=<0 0 0>, Index=&SpawnGroup(), Num=::NewEnumerator2)
    Offset Mode=::NewEnumerator0
    Override Seed=false
    Rotation Mode=::NewEnumerator3
    Rotation Quaternion=$Particles.SkeletalMeshLocation.SampledRotation
    Shape Origin=$Particles.SkeletalMeshLocation.SampledPosition
    Shape Primitive=::NewEnumerator1
    Surface Only=false
    UseOwnerScale=true
   )
   &InitialMeshOrientation(Use Rotation Vector=false)
   &AddVelocity(
    Module.Use Falloff Distance=false
    Override Seed=false
    Velocity Mode=::NewEnumerator1
    Velocity Speed=&MergeHelper_FloatSwitch(
     0=0
     1=0
     2=&RandomRangeFloat(Minimum=200, Maximum=350)
     3=&RandomRangeFloat(Minimum=25, Maximum=50)
     Index=&SpawnGroup()
     Num=::NewEnumerator2
    )
   )
   &CameraOffset(Camera Offset Amount=&MergeHelper_FloatSwitch(0=15, 1=8, 2=0, 3=?, Index=&SpawnGroup(), Num=::NewEnumerator1))
  ]
  particleUpdateScript [
   &ParticleState()
   /SetVariables(
    Particles.Falling Dust Color:LinearColor=&MultiplyLinearColor(
     Linear Color=$Particles.Initial.Falling Dust Color
     Scale Factor=&MakeVector4FromVectorAndFloat(
      "Vector (RGB)"=&MakeVectorFromLinearColorRGB(Color=&MultiplyLinearColor(
        Linear Color=&ColorFromCurve(ColorCurve=DataInterface() {propertyValues {
           AlphaCurve {keys [<at 0 1 RCIM_Cubic>]}
           BlueCurve {keys [<at 0.0061 0.1042>, {interpMode=RCIM_Cubic, leaveTangent=1.7407, time=0.1571, value=0.4323}]}
           GreenCurve {keys [<at 0.0061 0.1042>, {interpMode=RCIM_Cubic, leaveTangent=1.7407, time=0.1571, value=0.4323}]}
           RedCurve {keys [<at 0.0061 0.1042>, {interpMode=RCIM_Cubic, leaveTangent=1.7407, time=0.1571, value=0.4323}]}
          }})
        Scale Factor=&MakeVector4_FromLinearColor(LinearColor=&FNTimeOfDayColor(
          Night Color=<rgba 0.3066 0.4742 0.8177 1>
          Morning Color=<rgba 1 0.3015 0.2916 1>
          Day Color=<rgba 1 1 1 1>
          Evening Color=<rgba 1 0.5153 0.1302 1>
         ))
       ))
      "Float (Alpha)"=&FloatFromCurve(
       FloatCurve=DataInterface() {propertyValues {Curve {keys [
           {interpMode=RCIM_Cubic, leaveTangentWeight=0.0962, tangentMode=RCTM_Break, time=0.05, value=0}
           {interpMode=RCIM_Cubic, tangentMode=RCTM_Break, time=0.2, value=1}
           {arriveTangent=-1.8058, interpMode=RCIM_Cubic, tangentMode=RCTM_Break, time=1, value=0}
          ]}}}
       Scale Curve=&RandomRangeFloat(Minimum=0.4, Maximum=0.7)
      )
     )
    )
    Particles.Falling Dust Dynamic 1:Vector4f=&MakeVector4(
     X=&RandomRangeFloat(Minimum=0, Maximum=1)
     Y=&RandomRangeFloat(Minimum=0.5, Maximum=1)
     Z=3
     W=&FloatFromCurve(FloatCurve=DataInterface() {propertyValues {Curve {keys [<at 0 1 RCIM_Cubic>, {arriveTangent=-1.6019, interpMode=RCIM_Cubic, tangentMode=RCTM_Break, time=1, value=0}]}}})
    )
    Particles.Falling Dust Size:Vector2f=&Multiply_Vector2D(
     A=$Particles.Initial.Falling Dust Size
     B=&Multiply_Vector2D_ByFloat(
      Float=1
      Vector2D=&MakeVector2D(
       X=&FloatFromCurve(FloatCurve=DataInterface() {propertyValues {Curve {keys [
            {arriveTangent=1.3188, interpMode=RCIM_Cubic, time=0, value=0.2}
            {arriveTangent=0.7857, interpMode=RCIM_Cubic, leaveTangent=0.7857, tangentMode=RCTM_User, time=1, value=0.5}
           ]}}})
       Y=&FloatFromCurve(FloatCurve=DataInterface() {propertyValues {Curve {keys [
            {interpMode=RCIM_Cubic, tangentMode=RCTM_User, time=0, value=1.2}
            {arriveTangent=-2.3605, interpMode=RCIM_Cubic, leaveTangent=-2.3605, tangentMode=RCTM_User, time=1, value=0}
           ]}}})
      )
     )
    )
   ) {isSetParameters=true, module=SetVariables_#7}
   /SetVariables(
    Particles.Dust Color:LinearColor=&MultiplyLinearColor(
     Linear Color=$Particles.Initial.Dust Color
     Scale Factor=&MakeVector4FromVectorAndFloat(
      "Vector (RGB)"=<1 1 1>
      "Float (Alpha)"=&FloatFromCurve(FloatCurve=DataInterface() {propertyValues {Curve {keys [<at 0 0 RCIM_Cubic>,<at 0.3 1 RCIM_Cubic>,<at 1 0 RCIM_Cubic>]}}})
     )
    )
    Particles.Dust Dynamic 1:Vector4f=&MakeVector4(X=0, Y=1, Z=0, W=0)
    Particles.Dust Size:Vector2f=&Multiply_Vector2D(
     A=$Particles.Initial.Dust Size
     B=&Vector2DFromFloat(Value=&FloatFromCurve(FloatCurve=DataInterface() {propertyValues {Curve {keys [<at 0 0 RCIM_Cubic>, {arriveTangent=1, interpMode=RCIM_Cubic, leaveTangent=1, time=0.2, value=1}, <at 1 1 RCIM_Cubic>]}}}))
    )
    Particles.Dust SubImageIndex:NiagaraFloat=&FloatFromCurve(FloatCurve=DataInterface() {propertyValues {Curve {keys [<at 0 48>,<at 1 64>]}}})
   ) {isSetParameters=true, module=SetVariables_#8}
   &GravityForce(Gravity=&MergeHelper_VectorSwitch(
     0=<0 0 0>
     1=&RandomRangeVector(Minimum=<0 0 -50>, Maximum=<0 0 -100>)
     2=&RandomRangeVector(Minimum=<0 0 -680>, Maximum=<0 0 -980>)
     3=?
     Index=&SpawnGroup()
     Num=::NewEnumerator1
    ))
   &ScaleMeshSize(Scale Factor=&VectorFromFloat(Value=&FloatFromCurve(FloatCurve=DataInterface() {propertyValues {Curve {keys [
          {arriveTangent=-0.008, interpMode=RCIM_Cubic, leaveTangent=-0.008, tangentMode=RCTM_User, time=0, value=1}
          {arriveTangent=-3.2668, interpMode=RCIM_Cubic, leaveTangent=-3.2668, tangentMode=RCTM_User, time=1, value=0}
         ]}}})))
   &Collision(
    Age Colliding Particles=false
    Analytical Collision Plane Position 1:Vector3f=<0 0 -20>
    Analytical Collision Plane Space:ENiagaraCoordinateSpace=::Local
    CPU Collision Type=::NewEnumerator3
    Collision Enabled=&Boolean_AndOperation(
     Bool1=&IsPlatformSetActive(New Platform Set=DataInterface() {propertyValues {Platforms {qualityLevelMask=-28}}})
     Bool2=&SetBoolByFloatComparison(
      A=&FloatFromCurve(
       FloatCurve=DataInterface() {propertyValues {Curve {
          defaultValue=1
          keys [{interpMode=RCIM_Cubic, tangentMode=RCTM_User, time=0.0045, value=0}, {interpMode=RCIM_Cubic, tangentMode=RCTM_User, time=0.1114, value=1}]
         }}}
       Scale Curve=&MergeHelper_FloatSwitch(0=0, 1=0, 2=1, 3=?, Index=&SpawnGroup(), Num=::NewEnumerator1)
      )
      B=0.5
     )
    )
    KillOnCollision=false
    Module.Control Rest State Via Surface Normal=false
    Module.EnableMaxCollisionCount=false
    Module.Randomize Collision Normal Vector=false
    Perform 2 Analytical Collision Queries=false
   ) {module=Collision001}
   &Collision(
    Age Colliding Particles=false
    Collision Enabled=&Boolean_AndOperation(
     Bool1=&IsPlatformSetActive(New Platform Set=DataInterface() {propertyValues {Platforms {qualityLevelMask=-8}}})
     Bool2=&SetBoolByFloatComparison(
      A=&FloatFromCurve(
       FloatCurve=DataInterface() {propertyValues {Curve {
          defaultValue=1
          keys [{interpMode=RCIM_Cubic, tangentMode=RCTM_User, time=0.0045, value=0}, {interpMode=RCIM_Cubic, tangentMode=RCTM_User, time=0.1114, value=1}]
         }}}
       Scale Curve=&MergeHelper_FloatSwitch(0=0, 1=0, 2=1, 3=?, Index=&SpawnGroup(), Num=::NewEnumerator1)
      )
      B=0.5
     )
    )
    KillOnCollision=false
    Module.Control Rest State Via Surface Normal=false
    Module.EnableMaxCollisionCount=false
    Module.Randomize Collision Normal Vector=false
    Radius Calculation Type=::NewEnumerator1
   )
   &Drag(
    Drag=&MergeHelper_FloatSwitch(0=0, 1=1, 2=0, 3=1, Index=&SpawnGroup(), Num=::NewEnumerator2)
    Use Linear Drag=true
    Use Rotational Drag=true
   )
   &SolveForcesAndVelocity(Clamp Velocity=false, Limit Acceleration=false)
  ]
  rendererClasses [&NiagaraSpriteRendererProperties, &NiagaraMeshRendererProperties]
  renderers [
   {rendererClass=&NiagaraSpriteRendererProperties}
   {rendererClass=&NiagaraSpriteRendererProperties, rendererIndex=1}
   {rendererClass=&NiagaraSpriteRendererProperties, rendererIndex=2}
   {rendererClass=&NiagaraMeshRendererProperties, rendererIndex=3}
   {rendererClass=&NiagaraSpriteRendererProperties, rendererIndex=4}
  ]
 }
 wing_fingerB_02_r001 {
  emitterUpdateScript [
   &EmitterState(Life Cycle Mode=::NewEnumerator1, Loop Behavior=::NewEnumerator1, Loop Delay=2.582, UseLoopDelay=true)
   &SpawnRate(
    SpawnRate=&FloatFromCurve(
     FloatCurve=DataInterface() {propertyValues {Curve {keys [
         {arriveTangent=-0.3938, interpMode=RCIM_Cubic, time=2.922, value=1}
         {arriveTangent=-2.3758, interpMode=RCIM_Cubic, leaveTangent=-2.3758, time=3.2005, value=0.7429}
         <at 3.3429 0 RCIM_Cubic>
        ]}}}
     CurveIndex=$Emitter.Age
     Scale Curve=15
    )
    Use Spawn Probability=false
   )
   &SpawnBurst_Instantaneous(Spawn Count=8, Spawn Group=1, Use Spawn Probability=false, UseLoopCountLimit=false) {module=SpawnBurst_Instantaneous003}
   &SpawnRate(
    SpawnGroup=4
    SpawnRate=&FloatFromCurve(
     FloatCurve=DataInterface() {propertyValues {Curve {keys [
         {arriveTangent=-0.3938, interpMode=RCIM_Cubic, time=2.922, value=1}
         {arriveTangent=-2.3758, interpMode=RCIM_Cubic, leaveTangent=-2.3758, time=3.2005, value=0.7429}
         <at 3.3429 0 RCIM_Cubic>
        ]}}}
     CurveIndex=$Emitter.Age
     Scale Curve=5
    )
    Use Spawn Probability=false
   ) {module=SpawnRate002}
   &SpawnBurst_Instantaneous(Spawn Count=18, Spawn Group=2, Spawn Time=1.34, Use Spawn Probability=false, UseLoopCountLimit=false) {enabled=false}
   &SpawnRate(
    SpawnGroup=2
    SpawnRate=&FloatFromCurve(
     FloatCurve=DataInterface() {propertyValues {Curve {keys [
         {arriveTangent=4.5806, interpMode=RCIM_Cubic, leaveTangent=4.5806, tangentMode=RCTM_User, time=2.242, value=0}
         <at 2.3939 0.9855 RCIM_Cubic>
         {arriveTangent=-10.321, interpMode=RCIM_Cubic, leaveTangent=-10.321, tangentMode=RCTM_User, time=2.549, value=0}
        ]}}}
     CurveIndex=$Emitter.Age
     Scale Curve=60
    )
    Use Spawn Probability=false
   ) {module=SpawnRate001}
   &SpawnBurst_Instantaneous(Spawn Count=10, Spawn Group=3, Spawn Time=1.3, Use Spawn Probability=false, UseLoopCountLimit=false) {enabled=false, module=SpawnBurst_Instantaneous004}
   &SpawnBurst_Instantaneous(Spawn Count=5, Use Spawn Probability=false, UseLoopCountLimit=false) {enabled=false, module=SpawnBurst_Instantaneous002}
   /SetVariables(Emitter.DepthFade=10) {isSetParameters=true, module=SetVariables_#2}
   &SpawnBurst_Instantaneous(Spawn Count=37, Spawn Time=1.682, Use Spawn Probability=false, UseLoopCountLimit=false) {enabled=false, module=SpawnBurst_Instantaneous001}
  ]
  particleSpawnScript [
   &PartitionParticles(Number Of Partitions=3) {enabled=false}
   /SetVariables(Particles.VisibilityTag=&SpawnGroup()) {isSetParameters=true, module=SetVariables_#3}
   &InitializeParticle(
    Color Maximum=<rgba 3.0094 4.1614 5 1>
    Color Minimum=<rgba 3 3 3 1>
    Color Mode=::NewEnumerator2
    Lifetime=&MergeHelper_FloatSwitch(
     0=&RandomRangeFloat(Minimum=0.4, Maximum=0.8)
     1=&RandomRangeFloat(Minimum=0.8, Maximum=1.1)
     2=&RandomRangeFloat(Minimum=0.5, Maximum=1)
     3=&RandomRangeFloat(Minimum=0.8, Maximum=1.1)
     4=&RandomRangeFloat(Minimum=0.5, Maximum=1)
     Index=&SpawnGroup()
     Num=::NewEnumerator3
    )
    Mesh Scale Max=<0.5 0.5 0.5>
    Mesh Scale Min=<0.1 0.1 0.1>
    Mesh Scale Mode=::NewEnumerator2
    Sprite Rotation Angle=&MergeHelper_FloatSwitch(
     0=180
     1=&RandomRangeFloat(Minimum=0, Maximum=360)
     2=?
     3=?
     Index=&SpawnGroup()
     Num=::NewEnumerator0
    )
    Sprite Rotation Mode=::NewEnumerator0
    Sprite Size Mode=::NewEnumerator3
    Uniform Sprite Size=1
    UsePositionOffset=false
    Write Material Random=true
   )
   /SetVariables(
    Particles.Falling Dust Color:LinearColor=&RandomRangeLinearColor(Minimum=<rgba 0.1771 0.1771 0.1771 1>, Maximum=<rgba 0.2448 0.2448 0.2448 1>)
    Particles.Falling Dust Size:Vector2f=&RandomRangeVector2D(Minimum=<20 36>, Maximum=<30 60>)
    Particles.Falling Dust SubImageIndex:NiagaraFloat=&RandomRangeFloat(Minimum=0, Maximum=3)
   ) {isSetParameters=true, module=SetVariables_#4}
   /SetVariables(
    Particles.Dust Color:LinearColor=<rgba 0.0885 0.0885 0.0885 0.25>
    Particles.Dust Size:Vector2f=&Vector2DFromFloat(Value=&MaskFloat(
      Mask Source=0
      Mask Parameter=?
      Mask=1
      Masked Float=&RandomRangeFloat(Minimum=12, Maximum=25)
      Passthrough Float=&RandomRangeFloat(Minimum=30, Maximum=50)
     ))
   ) {isSetParameters=true, module=SetVariables_#5}
   /SetVariables(Particles.MeshIndex=&RandomRangeInt(Minimum=0, Maximum=2)) {isSetParameters=true, module=SetVariables_#6}
   &SkeletalMeshLocation(
    Bone / Socket Index=5
    Bone Sampling Mode=::NewEnumerator4
    MaskSpawnGroup=false
    Position Sampling=::NewEnumerator2
    Skeletal Mesh=$System.Skeletal Mesh Reader
   )
   &ShapeLocation(
    Additional Quaternion Rotation=true
    MaskSpawnGroup=false
    Non Uniform Scale=<2 2.5 1>
    Offset=&MergeHelper_VectorSwitch(0=<0 0 0>, 1=<0 0 0>, 2=<0 0 0>, 3=<0 0 0>, Index=&SpawnGroup(), Num=::NewEnumerator2)
    Offset Mode=::NewEnumerator0
    Override Seed=false
    Rotation Mode=::NewEnumerator3
    Shape Origin=$Particles.SkeletalMeshLocation.SampledPosition
    Sphere Radius=5
    UseOwnerScale=true
   )
   &InitialMeshOrientation(Use Rotation Vector=false)
   &AddVelocity(
    Module.Use Falloff Distance=false
    Override Seed=false
    Velocity Mode=::NewEnumerator1
    Velocity Origin=$Particles.SkeletalMeshLocation.SampledPosition
    Velocity Speed=&MergeHelper_FloatSwitch(
     0=0
     1=0
     2=&RandomRangeFloat(Minimum=200, Maximum=350)
     3=&RandomRangeFloat(Minimum=25, Maximum=50)
     4=0
     Index=&SpawnGroup()
     Num=::NewEnumerator3
    )
   )
   &CameraOffset(Camera Offset Amount=&MergeHelper_FloatSwitch(0=15, 1=8, 2=0, 3=?, Index=&SpawnGroup(), Num=::NewEnumerator1))
  ]
  particleUpdateScript [
   &ParticleState()
   /SetVariables(
    Particles.Falling Dust Color:LinearColor=&MultiplyLinearColor(
     Linear Color=$Particles.Initial.Falling Dust Color
     Scale Factor=&MakeVector4FromVectorAndFloat(
      "Vector (RGB)"=&MakeVectorFromLinearColorRGB(Color=&MultiplyLinearColor(
        Linear Color=&ColorFromCurve(ColorCurve=DataInterface() {propertyValues {
           AlphaCurve {keys [<at 0 1 RCIM_Cubic>]}
           BlueCurve {keys [<at 0.0061 0.1042>, {interpMode=RCIM_Cubic, leaveTangent=1.7407, time=0.1571, value=0.4323}]}
           GreenCurve {keys [<at 0.0061 0.1042>, {interpMode=RCIM_Cubic, leaveTangent=1.7407, time=0.1571, value=0.4323}]}
           RedCurve {keys [<at 0.0061 0.1042>, {interpMode=RCIM_Cubic, leaveTangent=1.7407, time=0.1571, value=0.4323}]}
          }})
        Scale Factor=&MakeVector4_FromLinearColor(LinearColor=&FNTimeOfDayColor(
          Night Color=<rgba 0.3066 0.4742 0.8177 1>
          Morning Color=<rgba 1 0.3015 0.2916 1>
          Day Color=<rgba 1 1 1 1>
          Evening Color=<rgba 1 0.5153 0.1302 1>
         ))
       ))
      "Float (Alpha)"=&FloatFromCurve(
       FloatCurve=DataInterface() {propertyValues {Curve {keys [
           {interpMode=RCIM_Cubic, leaveTangentWeight=0.0962, tangentMode=RCTM_Break, time=0.05, value=0}
           {interpMode=RCIM_Cubic, tangentMode=RCTM_Break, time=0.2, value=1}
           {arriveTangent=-1.8058, interpMode=RCIM_Cubic, tangentMode=RCTM_Break, time=1, value=0}
          ]}}}
       Scale Curve=&RandomRangeFloat(Minimum=0.4, Maximum=0.7)
      )
     )
    )
    Particles.Falling Dust Dynamic 1:Vector4f=&MakeVector4(
     X=&RandomRangeFloat(Minimum=0, Maximum=1)
     Y=&RandomRangeFloat(Minimum=0.5, Maximum=1)
     Z=3
     W=&FloatFromCurve(FloatCurve=DataInterface() {propertyValues {Curve {keys [<at 0 1 RCIM_Cubic>, {arriveTangent=-1.6019, interpMode=RCIM_Cubic, tangentMode=RCTM_Break, time=1, value=0}]}}})
    )
    Particles.Falling Dust Size:Vector2f=&Multiply_Vector2D(
     A=$Particles.Initial.Falling Dust Size
     B=&Multiply_Vector2D_ByFloat(
      Float=1
      Vector2D=&MakeVector2D(
       X=&FloatFromCurve(FloatCurve=DataInterface() {propertyValues {Curve {keys [
            {arriveTangent=1.3188, interpMode=RCIM_Cubic, time=0, value=0.2}
            {arriveTangent=0.7857, interpMode=RCIM_Cubic, leaveTangent=0.7857, tangentMode=RCTM_User, time=1, value=0.5}
           ]}}})
       Y=&FloatFromCurve(FloatCurve=DataInterface() {propertyValues {Curve {keys [
            <at 0 0 RCIM_Cubic>
            {interpMode=RCIM_Cubic, tangentMode=RCTM_User, time=0.4, value=1.2}
            {arriveTangent=-2.3605, interpMode=RCIM_Cubic, leaveTangent=-2.3605, tangentMode=RCTM_User, time=1, value=0}
           ]}}})
      )
     )
    )
   ) {isSetParameters=true, module=SetVariables_#7}
   /SetVariables(
    Particles.Dust Color:LinearColor=&MultiplyLinearColor(
     Linear Color=$Particles.Initial.Dust Color
     Scale Factor=&MakeVector4FromVectorAndFloat(
      "Vector (RGB)"=<1 1 1>
      "Float (Alpha)"=&FloatFromCurve(FloatCurve=DataInterface() {propertyValues {Curve {keys [<at 0 0 RCIM_Cubic>,<at 0.3 1 RCIM_Cubic>,<at 1 0 RCIM_Cubic>]}}})
     )
    )
    Particles.Dust Dynamic 1:Vector4f=&MakeVector4(X=0, Y=1, Z=0, W=0)
    Particles.Dust Size:Vector2f=&Multiply_Vector2D(
     A=$Particles.Initial.Dust Size
     B=&Vector2DFromFloat(Value=&FloatFromCurve(FloatCurve=DataInterface() {propertyValues {Curve {keys [<at 0 0 RCIM_Cubic>, {arriveTangent=1, interpMode=RCIM_Cubic, leaveTangent=1, time=0.2, value=1}, <at 1 1 RCIM_Cubic>]}}}))
    )
    Particles.Dust SubImageIndex:NiagaraFloat=&FloatFromCurve(FloatCurve=DataInterface() {propertyValues {Curve {keys [<at 0 48>,<at 1 64>]}}})
   ) {isSetParameters=true, module=SetVariables_#8}
   &GravityForce(Gravity=&MergeHelper_VectorSwitch(
     0=<0 0 0>
     1=&RandomRangeVector(Minimum=<0 0 -50>, Maximum=<0 0 -100>)
     2=&RandomRangeVector(Minimum=<0 0 -680>, Maximum=<0 0 -980>)
     3=<0 0 0>
     4=&RandomRangeVector(Minimum=<0 0 -680>, Maximum=<0 0 -980>)
     Index=&SpawnGroup()
     Num=::NewEnumerator3
    ))
   &ScaleMeshSize(Scale Factor=&VectorFromFloat(Value=&FloatFromCurve(FloatCurve=DataInterface() {propertyValues {Curve {keys [
          {arriveTangent=-0.008, interpMode=RCIM_Cubic, leaveTangent=-0.008, tangentMode=RCTM_User, time=0, value=1}
          {arriveTangent=-3.2668, interpMode=RCIM_Cubic, leaveTangent=-3.2668, tangentMode=RCTM_User, time=1, value=0}
         ]}}})))
   &Collision(
    Age Colliding Particles=false
    Analytical Collision Plane Position 1:Vector3f=<0 0 -20>
    Analytical Collision Plane Space:ENiagaraCoordinateSpace=::Local
    CPU Collision Type=::NewEnumerator3
    Collision Enabled=&Boolean_AndOperation(
     Bool1=&IsPlatformSetActive(New Platform Set=DataInterface() {propertyValues {Platforms {qualityLevelMask=-28}}})
     Bool2=&SetBoolByFloatComparison(
      A=&FloatFromCurve(
       FloatCurve=DataInterface() {propertyValues {Curve {
          defaultValue=1
          keys [{interpMode=RCIM_Cubic, tangentMode=RCTM_User, time=0.0045, value=0}, {interpMode=RCIM_Cubic, tangentMode=RCTM_User, time=0.1114, value=1}]
         }}}
       Scale Curve=&MergeHelper_FloatSwitch(0=0, 1=0, 2=1, 3=?, Index=&SpawnGroup(), Num=::NewEnumerator1)
      )
      B=0.5
     )
    )
    KillOnCollision=false
    Module.Control Rest State Via Surface Normal=false
    Module.EnableMaxCollisionCount=false
    Module.Randomize Collision Normal Vector=false
    Perform 2 Analytical Collision Queries=false
   ) {module=Collision001}
   &Collision(
    Age Colliding Particles=false
    Collision Enabled=&Boolean_AndOperation(
     Bool1=&IsPlatformSetActive(New Platform Set=DataInterface() {propertyValues {Platforms {qualityLevelMask=-8}}})
     Bool2=&SetBoolByFloatComparison(
      A=&FloatFromCurve(
       FloatCurve=DataInterface() {propertyValues {Curve {
          defaultValue=1
          keys [{interpMode=RCIM_Cubic, tangentMode=RCTM_User, time=0.0045, value=0}, {interpMode=RCIM_Cubic, tangentMode=RCTM_User, time=0.1114, value=1}]
         }}}
       Scale Curve=&MergeHelper_FloatSwitch(0=0, 1=0, 2=1, 3=?, Index=&SpawnGroup(), Num=::NewEnumerator1)
      )
      B=0.5
     )
    )
    KillOnCollision=false
    Module.Control Rest State Via Surface Normal=false
    Module.EnableMaxCollisionCount=false
    Module.Randomize Collision Normal Vector=false
    Radius Calculation Type=::NewEnumerator1
   )
   &Drag(
    Drag=&MergeHelper_FloatSwitch(0=0, 1=1, 2=0, 3=1, 4=0, Index=&SpawnGroup(), Num=::NewEnumerator3)
    Use Linear Drag=true
    Use Rotational Drag=true
   )
   &SolveForcesAndVelocity(Clamp Velocity=false, Limit Acceleration=false)
  ]
  rendererClasses [&NiagaraSpriteRendererProperties, &NiagaraMeshRendererProperties]
  renderers [
   {rendererClass=&NiagaraSpriteRendererProperties}
   {rendererClass=&NiagaraSpriteRendererProperties, rendererIndex=1}
   {rendererClass=&NiagaraSpriteRendererProperties, rendererIndex=2}
   {rendererClass=&NiagaraMeshRendererProperties, rendererIndex=3}
   {rendererClass=&NiagaraSpriteRendererProperties, rendererIndex=4}
   {rendererClass=&NiagaraMeshRendererProperties, rendererIndex=5}
  ]
 }
 wing_fingerB_02_r002 {
  emitterUpdateScript [
   &EmitterState(Life Cycle Mode=::NewEnumerator1, Loop Behavior=::NewEnumerator1, Loop Delay=3.322, UseLoopDelay=true)
   &SpawnRate(
    SpawnRate=&FloatFromCurve(
     FloatCurve=DataInterface() {propertyValues {Curve {keys [
         {arriveTangent=-0.3938, interpMode=RCIM_Cubic, time=2.922, value=1}
         {arriveTangent=-2.3758, interpMode=RCIM_Cubic, leaveTangent=-2.3758, time=3.2005, value=0.7429}
         <at 3.3429 0 RCIM_Cubic>
        ]}}}
     CurveIndex=$Emitter.Age
     Scale Curve=15
    )
    Use Spawn Probability=false
   ) {enabled=false}
   &SpawnBurst_Instantaneous(Spawn Count=8, Spawn Group=1, Use Spawn Probability=false, UseLoopCountLimit=false) {module=SpawnBurst_Instantaneous003}
   &SpawnBurst_Instantaneous(Spawn Count=8, Spawn Group=1, Spawn Time=0.1, Use Spawn Probability=false, UseLoopCountLimit=false) {module=SpawnBurst_Instantaneous008}
   &SpawnBurst_Instantaneous(Spawn Count=8, Spawn Group=1, Spawn Time=0.2, Use Spawn Probability=false, UseLoopCountLimit=false) {enabled=false, module=SpawnBurst_Instantaneous007}
   &SpawnBurst_Instantaneous(Spawn Count=10, Spawn Group=3, Spawn Time=0.1, Use Spawn Probability=false, UseLoopCountLimit=false) {enabled=false, module=SpawnBurst_Instantaneous004}
   &SpawnRate(
    SpawnGroup=4
    SpawnRate=&FloatFromCurve(
     FloatCurve=DataInterface() {propertyValues {Curve {keys [
         {arriveTangent=-14.8951, interpMode=RCIM_Cubic, time=3.322, value=1}
         {arriveTangent=-7.6059, interpMode=RCIM_Cubic, leaveTangent=-7.6059, time=3.3793, value=0.7664}
         {interpMode=RCIM_Cubic, leaveTangent=-4.0748, time=3.4535, value=0}
        ]}}}
     CurveIndex=$Emitter.Age
     Scale Curve=5
    )
    Use Spawn Probability=false
   ) {enabled=false, module=SpawnRate002}
   &SpawnBurst_Instantaneous(Spawn Count=12, Spawn Group=2, Use Spawn Probability=false, UseLoopCountLimit=false)
   &SpawnRate(
    SpawnGroup=2
    SpawnRate=&FloatFromCurve(
     FloatCurve=DataInterface() {propertyValues {Curve {keys [
         {arriveTangent=4.5806, interpMode=RCIM_Cubic, leaveTangent=4.5806, tangentMode=RCTM_User, time=2.242, value=0}
         <at 2.3939 0.9855 RCIM_Cubic>
         {arriveTangent=-10.321, interpMode=RCIM_Cubic, leaveTangent=-10.321, tangentMode=RCTM_User, time=2.549, value=0}
        ]}}}
     CurveIndex=$Emitter.Age
     Scale Curve=60
    )
    Use Spawn Probability=false
   ) {enabled=false, module=SpawnRate001}
   &SpawnBurst_Instantaneous(Spawn Count=12, Spawn Group=2, Spawn Time=0.1, Use Spawn Probability=false, UseLoopCountLimit=false) {module=SpawnBurst_Instantaneous005}
   &SpawnBurst_Instantaneous(Spawn Count=12, Spawn Group=2, Spawn Time=0.3, Use Spawn Probability=false, UseLoopCountLimit=false) {module=SpawnBurst_Instantaneous006}
   &SpawnBurst_Instantaneous(Spawn Count=5, Use Spawn Probability=false, UseLoopCountLimit=false) {enabled=false, module=SpawnBurst_Instantaneous002}
   /SetVariables(Emitter.DepthFade=10) {isSetParameters=true, module=SetVariables_#2}
   &SpawnBurst_Instantaneous(Spawn Count=37, Spawn Time=1.682, Use Spawn Probability=false, UseLoopCountLimit=false) {enabled=false, module=SpawnBurst_Instantaneous001}
  ]
  particleSpawnScript [
   &PartitionParticles(Number Of Partitions=3)
   /SetVariables(Particles.VisibilityTag=&SpawnGroup()) {isSetParameters=true, module=SetVariables_#3}
   &InitializeParticle(
    Color Maximum=<rgba 3.0094 4.1614 5 1>
    Color Minimum=<rgba 3 3 3 1>
    Color Mode=::NewEnumerator2
    Lifetime=&MergeHelper_FloatSwitch(
     0=&RandomRangeFloat(Minimum=0.4, Maximum=0.8)
     1=&RandomRangeFloat(Minimum=0.8, Maximum=1.1)
     2=&RandomRangeFloat(Minimum=0.5, Maximum=1)
     3=&RandomRangeFloat(Minimum=0.8, Maximum=1.1)
     4=&RandomRangeFloat(Minimum=0.5, Maximum=1)
     Index=&SpawnGroup()
     Num=::NewEnumerator3
    )
    Mesh Scale Max=<0.5 0.5 0.5>
    Mesh Scale Min=<0.1 0.1 0.1>
    Mesh Scale Mode=::NewEnumerator2
    Sprite Rotation Angle=&MergeHelper_FloatSwitch(
     0=180
     1=&RandomRangeFloat(Minimum=0, Maximum=360)
     2=?
     3=?
     Index=&SpawnGroup()
     Num=::NewEnumerator0
    )
    Sprite Rotation Mode=::NewEnumerator0
    Sprite Size Mode=::NewEnumerator3
    Uniform Sprite Size=1
    UsePositionOffset=false
    Write Material Random=true
   )
   /SetVariables(
    Particles.Falling Dust Color:LinearColor=&RandomRangeLinearColor(Minimum=<rgba 0.1771 0.1771 0.1771 1>, Maximum=<rgba 0.2448 0.2448 0.2448 1>)
    Particles.Falling Dust Size:Vector2f=&RandomRangeVector2D(Minimum=<20 36>, Maximum=<30 60>)
    Particles.Falling Dust SubImageIndex:NiagaraFloat=&RandomRangeFloat(Minimum=0, Maximum=3)
   ) {isSetParameters=true, module=SetVariables_#4}
   /SetVariables(
    Particles.Dust Color:LinearColor=<rgba 0.0885 0.0885 0.0885 0.25>
    Particles.Dust Size:Vector2f=&Vector2DFromFloat(Value=&MaskFloat(
      Mask Source=0
      Mask Parameter=?
      Mask=1
      Masked Float=&RandomRangeFloat(Minimum=30, Maximum=50)
      Passthrough Float=&RandomRangeFloat(Minimum=30, Maximum=50)
     ))
   ) {isSetParameters=true, module=SetVariables_#5}
   /SetVariables(Particles.MeshIndex=&RandomRangeInt(Minimum=0, Maximum=2)) {isSetParameters=true, module=SetVariables_#6}
   &SkeletalMeshLocation(
    Bone / Socket Index=&SelectIntFromArray(Direct Array Index=$Particles.PartitionParticles.Partition, Int Selection Array=DataInterface() {propertyValues {IntData [5,6,7]}})
    Bone Sampling Mode=::NewEnumerator4
    MaskSpawnGroup=false
    Position Sampling=::NewEnumerator2
    Skeletal Mesh=$System.Skeletal Mesh Reader
   )
   &ShapeLocation(
    Additional Quaternion Rotation=true
    MaskSpawnGroup=false
    Non Uniform Scale=<2 2.5 1>
    Offset=&MergeHelper_VectorSwitch(0=<0 0 0>, 1=<0 0 0>, 2=<0 0 0>, 3=<0 0 0>, Index=&SpawnGroup(), Num=::NewEnumerator2)
    Offset Mode=::NewEnumerator0
    Override Seed=false
    Rotation Mode=::NewEnumerator3
    Shape Origin=$Particles.SkeletalMeshLocation.SampledPosition
    Sphere Radius=5
    UseOwnerScale=true
   )
   &InitialMeshOrientation(Use Rotation Vector=false)
   &AddVelocity(
    Module.Use Falloff Distance=false
    Override Seed=false
    Velocity Mode=::NewEnumerator1
    Velocity Origin=$Particles.SkeletalMeshLocation.SampledPosition
    Velocity Speed=&MergeHelper_FloatSwitch(
     0=0
     1=&RandomRangeFloat(Minimum=15, Maximum=35)
     2=&RandomRangeFloat(Minimum=100, Maximum=150)
     3=&RandomRangeFloat(Minimum=25, Maximum=50)
     4=0
     Index=&SpawnGroup()
     Num=::NewEnumerator3
    )
   ) {enabled=false}
   &CameraOffset(Camera Offset Amount=&MergeHelper_FloatSwitch(0=0, 1=15, 2=0, 3=?, Index=&SpawnGroup(), Num=::NewEnumerator1))
   &AddVelocity(Velocity=&MergeHelper_VectorSwitch(
     0=<0 0 0>
     1=&RandomRangeVector(Minimum=<-20 -20 0>, Maximum=<20 20 20>)
     2=&RandomRangeVector(Minimum=<-150 -150 50>, Maximum=<150 150 100>)
     3=&RandomRangeVector(Minimum=<-20 -20 0>, Maximum=<20 20 20>)
     4=<0 0 0>
     Index=&SpawnGroup()
     Num=::NewEnumerator3
    )) {module=AddVelocity001}
  ]
  particleUpdateScript [
   &ParticleState()
   /SetVariables(
    Particles.Falling Dust Color:LinearColor=&MultiplyLinearColor(
     Linear Color=$Particles.Initial.Falling Dust Color
     Scale Factor=&MakeVector4FromVectorAndFloat(
      "Vector (RGB)"=&MakeVectorFromLinearColorRGB(Color=&MultiplyLinearColor(
        Linear Color=&ColorFromCurve(ColorCurve=DataInterface() {propertyValues {
           AlphaCurve {keys [<at 0 1 RCIM_Cubic>]}
           BlueCurve {keys [<at 0.0061 0.1042>, {interpMode=RCIM_Cubic, leaveTangent=1.7407, time=0.1571, value=0.4323}]}
           GreenCurve {keys [<at 0.0061 0.1042>, {interpMode=RCIM_Cubic, leaveTangent=1.7407, time=0.1571, value=0.4323}]}
           RedCurve {keys [<at 0.0061 0.1042>, {interpMode=RCIM_Cubic, leaveTangent=1.7407, time=0.1571, value=0.4323}]}
          }})
        Scale Factor=&MakeVector4_FromLinearColor(LinearColor=&FNTimeOfDayColor(
          Night Color=<rgba 0.3066 0.4742 0.8177 1>
          Morning Color=<rgba 1 0.3015 0.2916 1>
          Day Color=<rgba 1 1 1 1>
          Evening Color=<rgba 1 0.5153 0.1302 1>
         ))
       ))
      "Float (Alpha)"=&FloatFromCurve(
       FloatCurve=DataInterface() {propertyValues {Curve {keys [
           {interpMode=RCIM_Cubic, leaveTangentWeight=0.0962, tangentMode=RCTM_Break, time=0.05, value=0}
           {interpMode=RCIM_Cubic, tangentMode=RCTM_Break, time=0.2, value=1}
           {arriveTangent=-1.8058, interpMode=RCIM_Cubic, tangentMode=RCTM_Break, time=1, value=0}
          ]}}}
       Scale Curve=&RandomRangeFloat(Minimum=0.4, Maximum=0.7)
      )
     )
    )
    Particles.Falling Dust Dynamic 1:Vector4f=&MakeVector4(
     X=&RandomRangeFloat(Minimum=0, Maximum=1)
     Y=&RandomRangeFloat(Minimum=0.5, Maximum=1)
     Z=3
     W=&FloatFromCurve(FloatCurve=DataInterface() {propertyValues {Curve {keys [<at 0 1 RCIM_Cubic>, {arriveTangent=-1.6019, interpMode=RCIM_Cubic, tangentMode=RCTM_Break, time=1, value=0}]}}})
    )
    Particles.Falling Dust Size:Vector2f=&Multiply_Vector2D(
     A=$Particles.Initial.Falling Dust Size
     B=&Multiply_Vector2D_ByFloat(
      Float=1
      Vector2D=&MakeVector2D(
       X=&FloatFromCurve(FloatCurve=DataInterface() {propertyValues {Curve {keys [
            {arriveTangent=1.3188, interpMode=RCIM_Cubic, time=0, value=0.2}
            {arriveTangent=0.7857, interpMode=RCIM_Cubic, leaveTangent=0.7857, tangentMode=RCTM_User, time=1, value=0.5}
           ]}}})
       Y=&FloatFromCurve(FloatCurve=DataInterface() {propertyValues {Curve {keys [
            <at 0 0 RCIM_Cubic>
            {interpMode=RCIM_Cubic, tangentMode=RCTM_User, time=0.4, value=1.2}
            {arriveTangent=-2.3605, interpMode=RCIM_Cubic, leaveTangent=-2.3605, tangentMode=RCTM_User, time=1, value=0}
           ]}}})
      )
     )
    )
   ) {isSetParameters=true, module=SetVariables_#7}
   /SetVariables(
    Particles.Dust Color:LinearColor=&MultiplyLinearColor(
     Linear Color=$Particles.Initial.Dust Color
     Scale Factor=&MakeVector4FromVectorAndFloat(
      "Vector (RGB)"=<1 1 1>
      "Float (Alpha)"=&FloatFromCurve(FloatCurve=DataInterface() {propertyValues {Curve {keys [<at 0 0 RCIM_Cubic>,<at 0.3 1 RCIM_Cubic>,<at 1 0 RCIM_Cubic>]}}})
     )
    )
    Particles.Dust Dynamic 1:Vector4f=&MakeVector4(X=0, Y=1, Z=0, W=0)
    Particles.Dust Size:Vector2f=&Multiply_Vector2D(
     A=$Particles.Initial.Dust Size
     B=&Vector2DFromFloat(Value=&FloatFromCurve(FloatCurve=DataInterface() {propertyValues {Curve {keys [<at 0 0 RCIM_Cubic>, {arriveTangent=1, interpMode=RCIM_Cubic, leaveTangent=1, time=0.2, value=1}, <at 1 1 RCIM_Cubic>]}}}))
    )
    Particles.Dust SubImageIndex:NiagaraFloat=&FloatFromCurve(FloatCurve=DataInterface() {propertyValues {Curve {keys [<at 0 48>,<at 1 64>]}}})
   ) {isSetParameters=true, module=SetVariables_#8}
   &GravityForce(Gravity=&MergeHelper_VectorSwitch(
     0=<0 0 0>
     1=<0 0 0>
     2=&RandomRangeVector(Minimum=<0 0 -680>, Maximum=<0 0 -980>)
     3=<0 0 0>
     4=&RandomRangeVector(Minimum=<0 0 -680>, Maximum=<0 0 -980>)
     Index=&SpawnGroup()
     Num=::NewEnumerator3
    ))
   &ScaleMeshSize(Scale Factor=&VectorFromFloat(Value=&FloatFromCurve(FloatCurve=DataInterface() {propertyValues {Curve {keys [
          {arriveTangent=-0.008, interpMode=RCIM_Cubic, leaveTangent=-0.008, tangentMode=RCTM_User, time=0, value=1}
          {arriveTangent=-3.2668, interpMode=RCIM_Cubic, leaveTangent=-3.2668, tangentMode=RCTM_User, time=1, value=0}
         ]}}})))
   &Collision(
    Age Colliding Particles=false
    Analytical Collision Plane Position 1:Vector3f=<0 0 -20>
    Analytical Collision Plane Space:ENiagaraCoordinateSpace=::Local
    CPU Collision Type=::NewEnumerator3
    Collision Enabled=&Boolean_AndOperation(
     Bool1=&IsPlatformSetActive(New Platform Set=DataInterface() {propertyValues {Platforms {qualityLevelMask=-28}}})
     Bool2=&SetBoolByFloatComparison(
      A=&FloatFromCurve(
       FloatCurve=DataInterface() {propertyValues {Curve {
          defaultValue=1
          keys [{interpMode=RCIM_Cubic, tangentMode=RCTM_User, time=0.0045, value=0}, {interpMode=RCIM_Cubic, tangentMode=RCTM_User, time=0.1114, value=1}]
         }}}
       Scale Curve=&MergeHelper_FloatSwitch(0=0, 1=0, 2=1, 3=?, Index=&SpawnGroup(), Num=::NewEnumerator1)
      )
      B=0.5
     )
    )
    KillOnCollision=false
    Module.Control Rest State Via Surface Normal=false
    Module.EnableMaxCollisionCount=false
    Module.Randomize Collision Normal Vector=false
    Perform 2 Analytical Collision Queries=false
   ) {module=Collision001}
   &Collision(
    Age Colliding Particles=false
    Collision Enabled=&Boolean_AndOperation(
     Bool1=&IsPlatformSetActive(New Platform Set=DataInterface() {propertyValues {Platforms {qualityLevelMask=-8}}})
     Bool2=&SetBoolByFloatComparison(
      A=&FloatFromCurve(
       FloatCurve=DataInterface() {propertyValues {Curve {
          defaultValue=1
          keys [{interpMode=RCIM_Cubic, tangentMode=RCTM_User, time=0.0045, value=0}, {interpMode=RCIM_Cubic, tangentMode=RCTM_User, time=0.1114, value=1}]
         }}}
       Scale Curve=&MergeHelper_FloatSwitch(0=0, 1=0, 2=1, 3=?, Index=&SpawnGroup(), Num=::NewEnumerator1)
      )
      B=0.5
     )
    )
    KillOnCollision=false
    Module.Control Rest State Via Surface Normal=false
    Module.EnableMaxCollisionCount=false
    Module.Randomize Collision Normal Vector=false
    Radius Calculation Type=::NewEnumerator1
   )
   &Drag(
    Drag=&MergeHelper_FloatSwitch(0=0, 1=1, 2=0, 3=1, 4=0, Index=&SpawnGroup(), Num=::NewEnumerator3)
    Use Linear Drag=true
    Use Rotational Drag=true
   )
   &SolveForcesAndVelocity(Clamp Velocity=false, Limit Acceleration=false)
  ]
  rendererClasses [&NiagaraSpriteRendererProperties, &NiagaraMeshRendererProperties]
  renderers [
   {rendererClass=&NiagaraSpriteRendererProperties}
   {rendererClass=&NiagaraSpriteRendererProperties, rendererIndex=1}
   {rendererClass=&NiagaraSpriteRendererProperties, rendererIndex=2}
   {rendererClass=&NiagaraMeshRendererProperties, rendererIndex=3}
   {rendererClass=&NiagaraSpriteRendererProperties, rendererIndex=4}
   {rendererClass=&NiagaraMeshRendererProperties, rendererIndex=5}
  ]
 }
}
summary {
 systemName=NS_Emote_RockCrest_Debris
 emitters [
  {emitterName=Elbow_fr_r, rendererClasses [&NiagaraSpriteRendererProperties, &NiagaraMeshRendererProperties]}
  {
   emitterName=wing_fingerB_02_l
   rendererClasses [&NiagaraSpriteRendererProperties, &NiagaraMeshRendererProperties]
  }
  {
   emitterName=wing_fingerB_02_r
   rendererClasses [&NiagaraSpriteRendererProperties, &NiagaraMeshRendererProperties]
  }
  {
   emitterName=wing_fingerB_02_r001
   rendererClasses [&NiagaraSpriteRendererProperties, &NiagaraMeshRendererProperties]
  }
  {
   emitterName=wing_fingerB_02_r002
   rendererClasses [&NiagaraSpriteRendererProperties, &NiagaraMeshRendererProperties]
  }
 ]
 userVariables [
  User.bIsFrontEnd {type=NiagaraBool}
  User.bIsFrontEndPreview {type=NiagaraBool}
 ]
}
